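# Niagara system for the fire effect of the CinderMax main-player emote, built from flames clustered at seven fixed offsets.
# FireLarge bursts persistent flame meshes with waveform-flickered orange lights, while FireSmall spawns shorter-lived flame meshes in rings around matching offsets.
# FireBits emits sprites from those points that rise with drag and curl noise.
# Overhead Smoke meshes adjust opacity by vertical camera angle and tint by time of day.
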NiagaraSystem /BRCosmetics/Animation/Game/MainPlayer/Emotes/CinderMax/NS_Emote_CinderMax_Fire
dependencies {
 usedDataInterfaces [
  &NiagaraDataInterfaceArrayFloat3
  &NiagaraDataInterfaceArrayFloat
  &NiagaraDataInterfaceMeshRendererInfo
  &NiagaraDataInterfaceArrayInt32
  &NiagaraDataInterfaceCurve
  &NiagaraDataInterfaceCamera
 ]
 usedDynamicInputs []
 usedModules [&SystemState]
}
emitters {
 FireBits {
  emitterUpdateScript [
   &EmitterState()
   &SpawnRate(SpawnRate=10, Use Spawn Probability=false)
  ]
  particleSpawnScript [
   &PartitionParticles(Number Of Partitions=7)
   &InitializeParticle(
    Lifetime=&RandomRangeFloat(Minimum=0.8, Maximum=1.2)
    Mesh Scale Mode=::NewEnumerator1
    Mesh Uniform Scale Max=1.5
    Mesh Uniform Scale Min=0.5
    Position Offset=&Multiply_VectorByFloat(
     Vector=&SelectVectorFromArray(
      Direct Array Index=$Particles.PartitionParticles.Partition
      Vector Selection Array=DataInterface() {propertyValues {FloatData [<100 0 0>, <0 100 0>, <-100 0 0>, <0 -100 0>, <50 75 0>, <-50 -75 0>, <-75 60 0>]}}
     )
     Float=0.9
    )
    Sprite Size Max=<15 30>
    Sprite Size Min=<10 25>
    Sprite Size Mode=::NewEnumerator2
    UsePositionOffset=true
    Write Material Random=true
   )
   &DynamicMaterialParameters(
    Index 0 Param 1=&SelectFloatFromArray(Direct Array Index=$Particles.SpawnGroup, Float Selection Array=DataInterface() {propertyValues {FloatData [0,0.3]}})
    Module.Param0WriteEnabled=true
    Module.Param1WriteEnabled=true
    Module.Param2WriteEnabled=true
    Module.Param3WriteEnabled=true
   ) {enabled=false}
   &ShapeLocation(
    MaskSpawnGroup=false
    Offset=<0 0 -15>
    Offset Mode=::NewEnumerator0
    Ring / Disc Distribution Mode:ENiagaraSphereDistributionMode=::NewEnumerator3
    Ring Radius=150
    Shape Primitive=::NewEnumerator4
    Uniform Spiral Falloff=0
    UseOwnerScale=true
   ) {enabled=false}
   &AddVelocity(Velocity Speed Scale=2.5)
   &InitialMeshOrientation(
    Mesh Orientation Mode=::NewEnumerator2
    Rotation=&RandomRangeVector(Minimum=<-1 0 0>, Maximum=<1 0 0>)
    Use Rotation Vector=false
   ) {enabled=false}
  ]
  particleUpdateScript [
   &ParticleState()
   &ScaleMeshSize(Scale Factor=&VectorFromFloat(Value=&FloatFromCurve(FloatCurve=DataInterface() {propertyValues {Curve {keys [
          {arriveTangent=4.1106, interpMode=RCIM_Cubic, leaveTangent=4.1106, tangentMode=RCTM_User, time=0, value=0}
          {interpMode=RCIM_Cubic, tangentMode=RCTM_User, time=0.35, value=1}
          {tangentMode=RCTM_User, time=1, value=0}
         ]}}})))
   &UpdateMeshOrientation(Rotation Rate=0.25) {enabled=false}
   &SpriteRotationRate(Rotation Rate=100) {enabled=false}
   &ScaleSpriteSize(Uniform Curve Sprite Scale=DataInterface() {propertyValues {Curve {keys [
        {arriveTangent=4.677, interpMode=RCIM_Cubic, leaveTangent=4.677, tangentMode=RCTM_User, time=0, value=0}
        {interpMode=RCIM_Cubic, tangentMode=RCTM_User, time=0.3, value=1}
        {tangentMode=RCTM_User, time=1, value=0}
       ]}}})
   &DynamicMaterialParameters(
    Index 0 Param 1=&RandomRangeFloat(Minimum=0, Maximum=1)
    Module.Param0WriteEnabled=true
    Module.Param1WriteEnabled=true
    Module.Param2WriteEnabled=true
    Module.Param3WriteEnabled=true
   ) {module=DynamicMaterialParameters001}
   &GravityForce(Gravity=<0 0 200>)
   &Drag(Drag=2, Use Linear Drag=true, Use Rotational Drag=true)
   &CurlNoiseForce(Noise Frequency=10, Noise Strength=150, Pan Noise=true, Pan Noise Field=<1 0 0>)
   &SolveForcesAndVelocity(Clamp Velocity=false, Limit Acceleration=false)
   &FN_Sprite_Attributes(
    Color=<rgba 0.2691 0.2691 0.2691 1>
    Create Color Attribute=true
    Create SpriteSize Attribute=true
    SpriteSize=&Multiply_Vector2D_ByFloat(Float=1.15, Vector2D=$Particles.SpriteSize)
    Write to Module Namespace=true
   )
   /SetVariables(Particles.R2_CamOffset=-5) {isSetParameters=true}
   &FN_Mesh_Attributes(
    Color=<rgba 0.2691 0.2691 0.2691 1>
    Create Color Attribute=true
    Create Scale Attribute=true
    Scale=&Multiply_VectorByFloat(Vector=$Particles.Scale, Float=1.2)
    Write to Module Namespace=true
   ) {enabled=false}
  ]
  rendererClasses [&NiagaraSpriteRendererProperties]
  renderers [
   {rendererClass=&NiagaraSpriteRendererProperties}
   {rendererClass=&NiagaraSpriteRendererProperties, rendererIndex=1}
  ]
 }
 FireLarge {
  emitterUpdateScript [
   &EmitterState(
    Inactive Response=::NewEnumerator1
    Life Cycle Mode=::NewEnumerator1
    Loop Behavior=::NewEnumerator1
    Loop Duration Mode=::NewEnumerator1
    UseLoopDelay=false
   )
   &SpawnBurst_Instantaneous(Spawn Count=7, Use Spawn Probability=false, UseLoopCountLimit=false)
  ]
  particleSpawnScript [
   &PartitionParticles(Number Of Partitions=7)
   &InitializeParticle(
    Mesh Renderer Array Visibility Mode=::NewEnumerator3
    Mesh Renderer Info=DataInterface() {propertyValues {MeshRenderer=&self:FireLarge_0.NiagaraMeshRendererProperties_0}}
    Mesh Scale Mode=::NewEnumerator3
    Mesh Uniform Scale=&Multiply_Float(
     A=&SelectFloatFromArray(
      Direct Array Index=$Particles.PartitionParticles.Partition
      Float Selection Array=DataInterface() {propertyValues {FloatData [0.55,0.45,0.55,0.5,0.29,0.32,0.3]}}
     )
     B=0.8
    )
    Position Offset=&Multiply_VectorByFloat(
     Vector=&SelectVectorFromArray(
      Direct Array Index=$Particles.PartitionParticles.Partition
      Vector Selection Array=DataInterface() {propertyValues {FloatData [<100 0 -4>, <0 100 -4>, <-100 0 -5>, <0 -100 -4>, <50 75 -2>, <-50 -75 -2>, <-75 60 -3>]}}
     )
     Float=0.9
    )
    UsePositionOffset=true
    Write Material Random=true
   )
   &FN_SetSpawnGroup(Number of Spawn Groups=1)
   &InitialMeshOrientation(Mesh Orientation Mode=::NewEnumerator2, Use Rotation Vector=false)
   &DynamicMaterialParameters(
    Index 0 Param 1=&SelectFloatFromArray(Direct Array Index=$Particles.SpawnGroup, Float Selection Array=DataInterface() {propertyValues {FloatData [0,0.3]}})
    Module.Param0WriteEnabled=true
    Module.Param1WriteEnabled=true
    Module.Param2WriteEnabled=true
    Module.Param3WriteEnabled=true
   ) {enabled=false}
   &ShapeLocation(
    MaskSpawnGroup=false
    Offset=<0 0 -15>
    Offset Mode=::NewEnumerator0
    Ring / Disc Distribution Mode:ENiagaraSphereDistributionMode=::NewEnumerator3
    Ring Radius=150
    Shape Primitive=::NewEnumerator4
    Uniform Spiral Falloff=0
    UseOwnerScale=true
   ) {enabled=false}
   /SetVariables(Particles.MeshIndex=&SelectIntFromArray(Direct Array Index=$Particles.PartitionParticles.Partition, Int Selection Array=DataInterface() {propertyValues {IntData [1,0,1,1,0,0,0]}})) {isSetParameters=true}
  ]
  particleUpdateScript [
   &ParticleState(Kill Particles When Lifetime Has Elapsed=false, Let Infinitely Lived Particles Die When Emitter Deactivates=true)
   /Niagara/Noise/Waveform(
    EnableExponent1=false
    Global Input Position:NiagaraFloat=&RandomRangeFloat(Minimum=0, Maximum=100)
    Global Phase:NiagaraFloat=$Engine.Time
    GlobalClampMax=false
    GlobalClampMin=false
    UseInput1=false
    UsePhase1=false
    "Waveform [1]":ENiagara_Waveforms=::NewEnumerator9
    "[1] Frequency"=5
   )
   &Waveform(
    EnableExponent1=false
    Global Amplitude Scale=0.25
    Global Input Position:NiagaraFloat=&RandomRangeFloat(Minimum=0, Maximum=100)
    Global Phase:NiagaraFloat=$Engine.Time
    GlobalClampMax=false
    GlobalClampMin=false
    UseInput1=false
    UsePhase1=false
    "Waveform [1]":ENiagara_Waveforms=::NewEnumerator4
    "[1] Frequency"=2
   ) {enabled=false, module=Waveform001}
   &Waveform(EnableExponent1=false, Global Phase:NiagaraFloat=$Engine.Time, GlobalClampMax=false, GlobalClampMin=false, UseInput1=false, UsePhase1=false) {enabled=false, module=Waveform002}
   &ScaleMeshSizeBySpeed(
    Max Scale Factor=<1 0.1 1>
    Sample Scale Factor By Curve=true
    Scale Factor Curve=DataInterface() {propertyValues {Curve {keys [{interpMode=RCIM_Cubic, tangentMode=RCTM_User, time=0, value=0}, {interpMode=RCIM_Cubic, tangentMode=RCTM_User, time=1, value=1}]}}}
   ) {enabled=false}
   &ScaleMeshSize(Scale Factor=&Multiply_VectorByFloat(
     Vector=&MakeVector(
      X=1
      Y=1
      Z=&Add_Float(
       A=1
       B=&Multiply_Float(
        A=&MakeFloatFromVector(Vector=$Particles.Scale, Channel=::NewEnumerator0)
        B=&Multiply_Float(A=$Particles.Waveform001.Waveform, B=0.8)
       )
      )
     )
     Float=1
    )) {enabled=false}
   &ScaleMeshSize(Scale Factor=&Multiply_VectorByFloat(
     Vector=&MakeVector(
      X=1
      Y=1
      Z=&Add_Float(
       A=1
       B=&Multiply_Float(
        A=&MakeFloatFromVector(Vector=$Particles.Scale, Channel=::NewEnumerator0)
        B=&Divide_Float(
         A=/Niagara/DynamicInputs/Integer/V2/Round_Float(Float:NiagaraFloat=&Multiply_Float(
           A=&ModuloFloat(A=&Add_Float(A=$Engine.Time, B=&RandomRangeFloat(Minimum=0, Maximum=1)), B=1)
           B=12
          ))
         B=12
        )
       )
      )
     )
     Float=1
    )) {enabled=false, module=ScaleMeshSize001}
   /Niagara/Modules/Update/Color/ScaleColorBySpeed(
    A Scale Max=0
    A Scale Min=1
    Max Speed Threshold=1000
    RGB Scale Min=1
    Source Velocity:Vector3f=$Particles.SkeletalMeshLocation.SampledVelocity
   ) {enabled=false}
   &InterpolateOverTime(Rate Of Change=10, Target Vector Value:NiagaraPosition=$Particles.Position, Variable Type=::NewEnumerator2)
   &DynamicMaterialParameters(
    Index 0 Param 1=&RandomRangeFloat(Minimum=0, Maximum=1)
    Module.Param0WriteEnabled=true
    Module.Param1WriteEnabled=true
    Module.Param2WriteEnabled=true
    Module.Param3WriteEnabled=true
   ) {module=DynamicMaterialParameters001}
   &FN_Light_Attributes(
    Color=<rgba 1 0.5266 0 1>
    Create Color Attribute=true
    Create Exponent Attribute=true
    Create Position Attribute=true
    Create Radius Attribute=true
    Enable Particle Light Enable=true
    Exponent=5
    Position=&AddVectorToPosition(Position=$Particles.Position, Vector=<0 0 50>)
    Radius=150
    Scale Color=&Add_Float(
     A=&Multiply_Float(A=&MakeFloatFromVector(Vector=$Particles.Scale, Channel=::NewEnumerator0), B=15)
     B=&Multiply_Float(
      A=&MakeFloatFromVector(Vector=$Particles.Scale, Channel=::NewEnumerator0)
      B=&Multiply_Float(A=$Particles.Waveform.Waveform, B=20)
     )
    )
   )
   &FN_Mesh_Attributes(
    Color=<rgba 0.2691 0.2691 0.2691 1>
    Create Color Attribute=true
    Create Scale Attribute=true
    Scale=&Multiply_VectorByFloat(Vector=$Particles.Scale, Float=1.065)
    Write to Module Namespace=true
   )
   /SetVariables(Particles.R2_CamOffset=-5) {isSetParameters=true}
  ]
  rendererClasses [&NiagaraMeshRendererProperties, &NiagaraLightRendererProperties]
  renderers [
   {rendererClass=&NiagaraMeshRendererProperties, rendererIndex=0}
   {rendererClass=&NiagaraLightRendererProperties, rendererIndex=1}
   {rendererClass=&NiagaraMeshRendererProperties, rendererIndex=2}
  ]
 }
 FireSmall {
  emitterUpdateScript [
   &EmitterState()
   &SpawnBurst_Instantaneous(Spawn Count=2, Use Spawn Probability=false, UseLoopCountLimit=false) {enabled=false}
   &SpawnRate(Use Spawn Probability=false)
  ]
  particleSpawnScript [
   &PartitionParticles(Number Of Partitions=7)
   &InitializeParticle(
    Lifetime=&RandomRangeFloat(Minimum=4, Maximum=6)
    Mesh Scale=&Multiply_VectorByFloat(
     Vector=<1 1 1>
     Float=&Multiply_Float(
      A=&SelectFloatFromArray(
       Direct Array Index=$Particles.PartitionParticles.Partition
       Float Selection Array=DataInterface() {propertyValues {FloatData [0.55,0.45,0.55,0.5,0.29,0.32,0.3]}}
      )
      B=0.8
     )
    )
    Mesh Scale Mode=::NewEnumerator4
    Position Offset=&Multiply_VectorByFloat(
     Vector=&SelectVectorFromArray(
      Direct Array Index=$Particles.PartitionParticles.Partition
      Vector Selection Array=DataInterface() {propertyValues {FloatData [<100 0 4>, <0 100 4>, <-100 0 4>, <0 -100 4>, <50 75 4>, <-50 -75 4>, <-75 60 4>]}}
     )
     Float=0.9
    )
    Sprite Size Max=<30 35>
    Sprite Size Min=<15 24>
    Sprite Size Mode=::NewEnumerator2
    UsePositionOffset=true
    Write Material Random=true
   )
   &ShapeLocation(
    MaskSpawnGroup=false
    Offset=<0 0 -3>
    Offset Mode=::NewEnumerator0
    Override Seed=false
    Ring Radius=45
    Shape Origin=$Particles.Position
    Shape Primitive=::NewEnumerator4
    UseOwnerScale=true
   )
  ]
  particleUpdateScript [
   &ParticleState()
   &ScaleMeshSize(Scale Factor=&VectorFromFloat(Value=&FloatFromCurve(FloatCurve=DataInterface() {propertyValues {Curve {keys [
          {arriveTangent=4.677, interpMode=RCIM_Cubic, leaveTangent=4.677, tangentMode=RCTM_User, time=0, value=0}
          {interpMode=RCIM_Cubic, tangentMode=RCTM_User, time=0.05, value=1}
          <at 0.95 1>
          {tangentMode=RCTM_User, time=1, value=0}
         ]}}})))
   &ScaleSpriteSize(Uniform Curve Sprite Scale=DataInterface() {propertyValues {Curve {keys [
        {arriveTangent=4.677, interpMode=RCIM_Cubic, leaveTangent=4.677, tangentMode=RCTM_User, time=0, value=0}
        {interpMode=RCIM_Cubic, tangentMode=RCTM_User, time=0.05, value=1}
        <at 0.95 1>
        {tangentMode=RCTM_User, time=1, value=0}
       ]}}}) {enabled=false}
   &DynamicMaterialParameters(
    Index 0 Param 1=&RandomRangeFloat(Minimum=0, Maximum=1)
    Module.Param0WriteEnabled=true
    Module.Param1WriteEnabled=true
    Module.Param2WriteEnabled=true
    Module.Param3WriteEnabled=true
   ) {module=DynamicMaterialParameters001}
   /SetVariables(Particles.R2_CamOffset=-5) {isSetParameters=true}
   &FN_Sprite_Attributes(
    Color=<rgba 0.2691 0.2691 0.2691 1>
    Create Color Attribute=true
    Create SpriteSize Attribute=true
    SpriteSize=&Multiply_Vector2D(A=$Particles.SpriteSize, B=<1.15 1.05>)
    Write to Module Namespace=true
   )
   &FN_Mesh_Attributes(
    Color=<rgba 0.2691 0.2691 0.2691 1>
    Create Color Attribute=true
    Create Scale Attribute=true
    Scale=&Multiply_VectorByFloat(Vector=$Particles.Scale, Float=1.065)
    Write to Module Namespace=true
   )
  ]
  rendererClasses [&NiagaraMeshRendererProperties]
  renderers [
   {rendererClass=&NiagaraMeshRendererProperties, rendererIndex=0}
   {rendererClass=&NiagaraMeshRendererProperties, rendererIndex=1}
  ]
 }
 Smoke {
  emitterUpdateScript [
   &EmitterState(
    Inactive Response=::NewEnumerator1
    Life Cycle Mode=::NewEnumerator1
    Loop Behavior=::NewEnumerator1
    Loop Duration Mode=::NewEnumerator1
    UseLoopDelay=false
   )
   &SpawnBurst_Instantaneous(Spawn Count=5, Use Spawn Probability=false, UseLoopCountLimit=false)
  ]
  particleSpawnScript [
   &InitializeParticle(
    Mesh Renderer Array Visibility Mode=::NewEnumerator3
    Mesh Renderer Info=DataInterface() {propertyValues {MeshRenderer=&self:Smoke_3.NiagaraMeshRendererProperties_0}}
    Mesh Scale Mode=::NewEnumerator1
    Mesh Uniform Scale Max=0.5
    Mesh Uniform Scale Min=0.4
    UsePositionOffset=false
    Write Material Random=true
   )
   &FN_SetSpawnGroup(Number of Spawn Groups=1)
   &InitialMeshOrientation(
    Mesh Axis To Orient=<0 1 0>
    Mesh Orientation Mode=::NewEnumerator2
    Orientation Vector=<0 0 1>
    Rotation=<0 0.25 0>
    Use Rotation Vector=true
   )
   &DynamicMaterialParameters(
    Index 0 Param 1=&SelectFloatFromArray(Direct Array Index=$Particles.SpawnGroup, Float Selection Array=DataInterface() {propertyValues {FloatData [0,0.3]}})
    Module.Param0WriteEnabled=true
    Module.Param1WriteEnabled=true
    Module.Param2WriteEnabled=true
    Module.Param3WriteEnabled=true
   ) {enabled=false}
   &ShapeLocation(
    MaskSpawnGroup=false
    Offset=<0 0 250>
    Offset Mode=::NewEnumerator0
    Ring / Disc Distribution Mode:ENiagaraSphereDistributionMode=::NewEnumerator3
    Shape Primitive=::NewEnumerator4
    Uniform Spiral Falloff=0.5
    UseOwnerScale=true
   )
  ]
  particleUpdateScript [
   &ParticleState(Kill Particles When Lifetime Has Elapsed=false, Let Infinitely Lived Particles Die When Emitter Deactivates=true)
   &ScaleMeshSizeBySpeed(
    Max Scale Factor=<1 0.1 1>
    Sample Scale Factor By Curve=true
    Scale Factor Curve=DataInterface() {propertyValues {Curve {keys [{interpMode=RCIM_Cubic, tangentMode=RCTM_User, time=0, value=0}, {interpMode=RCIM_Cubic, tangentMode=RCTM_User, time=1, value=1}]}}}
   ) {enabled=false}
   &ScaleMeshSize(Scale Factor=&MakeVector(
     X=&FloatFromCurve(FloatCurve=DataInterface() {propertyValues {Curve {keys [
          {arriveTangent=9.0725, interpMode=RCIM_Cubic, leaveTangent=9.0725, tangentMode=RCTM_User, time=0, value=0.15}
          {arriveTangent=-1.198, interpMode=RCIM_Cubic, leaveTangent=-1.198, tangentMode=RCTM_User, time=0.1408, value=0.9847}
          {arriveTangent=-0.0405, interpMode=RCIM_Cubic, leaveTangent=-0.0405, tangentMode=RCTM_User, time=1, value=0.725}
         ]}}})
     Y=&FloatFromCurve(FloatCurve=DataInterface() {propertyValues {Curve {keys [
          {arriveTangent=9.0725, interpMode=RCIM_Cubic, leaveTangent=9.0725, tangentMode=RCTM_User, time=0, value=0.2}
          {arriveTangent=-1.198, interpMode=RCIM_Cubic, leaveTangent=-1.198, tangentMode=RCTM_User, time=0.5, value=0.65}
          {arriveTangent=-0.0405, interpMode=RCIM_Cubic, leaveTangent=-0.0405, tangentMode=RCTM_User, time=1, value=0.725}
         ]}}})
     Z=1
    )) {enabled=false}
   &ScaleColorBySpeed(
    A Scale Max=0
    A Scale Min=1
    Max Speed Threshold=1000
    RGB Scale Min=1
    Source Velocity:Vector3f=$Particles.SkeletalMeshLocation.SampledVelocity
   ) {enabled=false}
   &InterpolateOverTime(Rate Of Change=10, Target Vector Value:NiagaraPosition=$Particles.Position, Variable Type=::NewEnumerator2) {enabled=false}
   &DynamicMaterialParameters(
    Index 0 Param 1=&RandomRangeFloat(Minimum=0, Maximum=1)
    Module.Param0WriteEnabled=true
    Module.Param1WriteEnabled=true
    Module.Param2WriteEnabled=true
    Module.Param3WriteEnabled=true
   ) {module=DynamicMaterialParameters001}
   /SetVariables(Particles.R2_CamOffset=-5) {isSetParameters=true}
   /SetVariables(Particles.TransparentFromTop:NiagaraFloat=&FloatFromCurve(
     FloatCurve=DataInterface() {propertyValues {Curve {keys [<at -1 0.25>,<at 0 1>]}}}
     CurveIndex=&DotProduct(Vector A=<0 0 1>, Vector B=&ReturnCameraVector(Camera Vectors=::NewEnumerator1, Coordinate Space=::Simulation))
    )) {isSetParameters=true}
   &FN_Mesh_Attributes(
    Color=&MultiplyLinearColor(
     Linear Color=<rgba 0.2934 0.2934 0.2934 1>
     Scale Factor=&MakeVector4(X=1, Y=1, Z=1, W=$Particles.TransparentFromTop)
    )
    Create Color Attribute=true
    Create Scale Attribute=true
    Scale=&Multiply_VectorByFloat(Vector=$Particles.Scale, Float=1.05)
    Write to Module Namespace=true
   )
   &ScaleColor(
    Scale Alpha=$Particles.TransparentFromTop
    Scale RGB=&MakeVectorFromLinearColorRGB(Color=&FN_Select_Color_from_Time_Of_Day(Morning=<rgba 1 1 1 1>, Day=<rgba 1 1 1 1>, Evening=<rgba 1 0.8022 0.4844 1>, Night=<rgba 1 0.7676 0.3941 1>))
    ScaleA=true
    ScaleRGB=true
   )
  ]
  rendererClasses [&NiagaraMeshRendererProperties]
  renderers [
   {rendererClass=&NiagaraMeshRendererProperties, rendererIndex=0}
   {rendererClass=&NiagaraMeshRendererProperties, rendererIndex=1}
  ]
 }
}
summary {
 systemName=NS_Emote_CinderMax_Fire
 emitters [
  {emitterName=FireLarge, rendererClasses [&NiagaraMeshRendererProperties, &NiagaraLightRendererProperties]}
  {emitterName=Smoke, rendererClasses [&NiagaraMeshRendererProperties]}
  {emitterName=FireBits, rendererClasses [&NiagaraSpriteRendererProperties]}
  {emitterName=FireSmall, rendererClasses [&NiagaraMeshRendererProperties]}
 ]
 userVariables [
  User.Skeletal Mesh {
   type=NiagaraDataInterfaceSkeletalMesh
   defaultValue {
    struct=NiagaraExt_VariableValue_DataInterface
    value {dataInterface=&self:NiagaraDataInterfaceSkeletalMesh_0}
   }
  }
  User.SkeletalMesh {
   type=NiagaraDataInterfaceSkeletalMesh
   defaultValue {
    struct=NiagaraExt_VariableValue_DataInterface
    value {dataInterface=&self:NiagaraDataInterfaceSkeletalMesh_1}
   }
  }
 ]
}
